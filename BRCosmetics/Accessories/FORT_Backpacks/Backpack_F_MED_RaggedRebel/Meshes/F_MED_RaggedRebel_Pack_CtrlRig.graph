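# Control Rig for the RaggedRebel backpack that spins its blades bone at speeds set by the character's movement state.
# Chained branches on movement flags, where later ones override earlier, write a speed, duration and blend vector.
# A timed loop samples that vector, updating next speed and blend in its second half and current speed and duration near its end.
# An accumulated lerp of those speeds drives a looping angle that rotates the bone.
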ControlRigBlueprint /BRCosmetics/Accessories/FORT_Backpacks/Backpack_F_MED_RaggedRebel/Meshes/F_MED_RaggedRebel_Pack_CtrlRig
graphs [
 RigVMFunctionLibrary {links [], nodes []}
 RigVMModel {nodes [
   RigUnit_BeginExecution = &RigUnit_BeginExecution()
   RigVMFunction_Sequence = &RigVMFunction_Sequence(ExecuteContext=@RigUnit_BeginExecution.ExecutePin)
   n1 = &RigVMFunction_ControlFlowBranch(ExecuteContext=@RigVMFunction_Sequence.A, Condition=RigVMVariableNode() {variable=Is IDLE}.Value)
   RigVMFunction_ControlFlowBranch = &RigVMFunction_ControlFlowBranch(ExecuteContext=@n1.Completed, Condition=RigVMVariableNode() {variable=Is Sprinting}.Value)
   n2 = &RigVMFunction_ControlFlowBranch(ExecuteContext=@RigVMFunction_ControlFlowBranch.Completed, Condition=RigVMVariableNode() {variable=Is Falling}.Value)
   n3 = &RigVMFunction_ControlFlowBranch(ExecuteContext=@n2.Completed, Condition=RigVMVariableNode() {variable=Is Skydive}.Value)
   n4 = &RigVMFunction_ControlFlowBranch(ExecuteContext=@n3.Completed, Condition=RigVMVariableNode() {variable=Is skydive dive}.Value)
   n5 = &RigVMFunction_ControlFlowBranch(ExecuteContext=@n4.Completed, Condition=RigVMVariableNode() {variable=Is skydive parachute}.Value)
   n6 = &RigVMFunction_ControlFlowBranch(ExecuteContext=@n5.Completed, Condition=RigVMVariableNode() {variable=Is Grindrail}.Value)
   n7 = &RigVMFunction_ControlFlowBranch(ExecuteContext=@n6.Completed, Condition=RigVMVariableNode() {variable=Is Tactical Sprint}.Value)
   n8 = &RigVMFunction_ControlFlowBranch(ExecuteContext=@n7.Completed, Condition=RigVMVariableNode() {variable=Is Zipline}.Value)
   n9 = &RigVMFunction_ControlFlowBranch(ExecuteContext=@n8.Completed, Condition=RigVMVariableNode() {variable=Is Going Up}.Value)
   n10 = &RigVMFunction_ControlFlowBranch(ExecuteContext=@n9.Completed, Condition=RigVMVariableNode() {variable=Is Going Down}.Value)
   RigVMVariableNode(Value=(X=700,Y=2,Z=5), ExecuteContext=@n10.True) {variable=Speed And Duration}
   RigVMVariableNode(Value=(X=700,Y=2,Z=5), ExecuteContext=@n9.True) {variable=Speed And Duration}
   RigVMVariableNode(Value=(X=1500,Y=2,Z=5), ExecuteContext=@n8.True) {variable=Speed And Duration}
   RigVMVariableNode(Value=(X=500,Y=2,Z=3), ExecuteContext=@n7.True) {variable=Speed And Duration}
   RigVMVariableNode(Value=(X=1000,Y=2,Z=3), ExecuteContext=@n6.True) {variable=Speed And Duration}
   RigVMVariableNode(Value=(X=700,Y=2,Z=3), ExecuteContext=@n5.True) {variable=Speed And Duration}
   RigVMVariableNode(Value=(X=1500,Y=1,Z=3), ExecuteContext=@n4.True) {variable=Speed And Duration}
   RigVMVariableNode(Value=(X=500,Y=1.5,Z=3), ExecuteContext=@n3.True) {variable=Speed And Duration}
   RigVMVariableNode(Value=(X=700,Y=2,Z=3), ExecuteContext=@n2.True) {variable=Speed And Duration}
   RigVMVariableNode(Value=(X=250,Y=2,Z=3), ExecuteContext=@RigVMFunction_ControlFlowBranch.True) {variable=Speed And Duration}
   RigVMVariableNode(Value=(X=0,Y=0.1,Z=2), ExecuteContext=@n1.True) {variable=Speed And Duration}
   n11 = RigVMVariableNode() {variable=Initial Duration}
   n12 = /Script/RigVM.RigVMFunction_TimeLoop(Normalize=false, Speed=1, Duration=@n11.Value) {is=struct RigVMUnitNode}
   n13 = &RigVMFunction_ControlFlowBranch(
    Condition=/Script/RigVM.RigVMFunction_MathDoubleGreaterEqual(A=@n12.Relative, B=&RigVMFunction_MathDoubleSub(B=0.1, A=@n11.Value).Result) {is=struct RigVMUnitNode, template="GreaterEqual::Execute(in A,in B,out Result)"}.Result
    ExecuteContext=@RigVMFunction_Sequence.B
   )
   Multiply = &RigVMFunction_MathDoubleMul(B=0.5, A=@n11.Value)
   n14 = &RigVMFunction_ControlFlowBranch(
    Condition=&RigVMFunction_MathDoubleGreaterEqual(B=@Multiply.Result, A=@n12.Relative) {template="GreaterEqual::Execute(in A,in B,out Result)"}.Result
    ExecuteContext=@n13.Completed
   )
   AccumulateLerp = /Script/RigVM.RigVMFunction_AccumulateFloatLerp(
    bIntegrateDeltaTime=true
    TargetValue=RigVMVariableNode() {variable=Current Speed}.Value
    Blend=RigVMVariableNode() {variable=blend state}.Value
    InitialValue=RigVMVariableNode() {variable=Next Speed}.Value
   ) {is=struct RigVMUnitNode, template="AccumulateLerp::Execute(in TargetValue,in InitialValue,in Blend,in bIntegrateDeltaTime,out Result)"}
   RigVMFunction_TimeLoop = &RigVMFunction_TimeLoop(Duration=359, Normalize=false, Speed=@AccumulateLerp.Result)
   RigVMFunction_MathQuaternionFromEuler = &RigVMFunction_MathQuaternionFromEuler(Euler=(X=0,Y=0,Z=0), Euler.Y=@RigVMFunction_TimeLoop.Relative)
   &RigUnit_ModifyTransforms(
    ItemToModify=((Item=(Type=Bone,Name="blades"),Transform=(Rotation=(X=0,Y=0,Z=0,W=1),Translation=(X=0,Y=0,Z=0),Scale3D=(X=1,Y=1,Z=1))))
    Weight=1
    WeightMinimum=0
    ExecutePin=@n14.Completed
    ItemToModify.0.Transform.Rotation=@RigVMFunction_MathQuaternionFromEuler.Result
   )
   n15 = RigVMVariableNode() {variable=Speed And Duration}
   RigVMVariableNode(ExecuteContext=@n14.True, Value=@n15.Value.X) {variable=Next Speed}
   > RigVMVariableNode(Value=@n15.Value.Z) {variable=blend state}
   RigVMVariableNode(ExecuteContext=@n13.True, Value=@n15.Value.Y) {variable=Initial Duration}
   > RigVMVariableNode(Value=@n15.Value.Y) {variable=Current Duration}
   > RigVMVariableNode(Value=@n15.Value.X) {variable=Current Speed}
   RigVMVariableNode(Value=RigVMVariableNode() {variable=Initial Duration}.Value, ExecuteContext=@n13.False) {variable=Initial Duration}
  ]}
]
hierarchy [
 root {loc=<0 0 0>, rot=<0 0 0 1>}
 pack {parent=root, loc=<0 -0.7316 0>, rot=<0 0 0 1>}
 body {parent=pack, loc=<0 0.3691 3.2419>, rot=<0 0 0.7071 -0.7071>}
 offset_blades {parent=body, loc=<7.4132 0 -3.2419>, rot=<0 0 -0.7071 -0.7071>}
 blades {parent=offset_blades, loc=<0 0 0>, rot=<0 0 0 1>}
]
variables [
 Current Duration {type=double, default=5}
 Current Speed {type=double, default=0}
 Initial Duration {type=double, default=1}
 Is Emote {type=bool, default=False}
 Is Falling {type=bool, default=False}
 Is Going Down {type=bool, default=False}
 Is Going Up {type=bool, default=False}
 Is Grindrail {type=bool, default=False}
 Is IDLE {type=bool, default=False}
 Is Skydive {type=bool, default=False}
 Is Skydive Dive {type=bool, default=False}
 Is Skydive Parachute {type=bool, default=False}
 Is Sprinting {type=bool, default=False}
 Is Tactical Sprint {type=bool, default=False}
 Is Zipline {type=bool, default=False}
 Next Speed {type=double, default=0}
 Speed And Duration {type=FVector, default=(X=0,Y=0,Z=0)}
 blend state {type=double, default=0}
]
