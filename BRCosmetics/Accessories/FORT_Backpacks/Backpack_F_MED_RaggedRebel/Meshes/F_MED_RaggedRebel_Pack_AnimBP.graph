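# Animation Blueprint for the Ragged Rebel backpack that passes its input pose through a Control Rig driven by player movement states.
# Each update derives idle, sprinting and going up or down, using vertical movement gated against airborne and zipline states.
# Anim input events maintain the zipline flag.
# These flags join skydive, parachute, falling, grind rail and tactical sprint states as rig inputs.
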AnimBlueprint /BRCosmetics/Accessories/FORT_Backpacks/Backpack_F_MED_RaggedRebel/Meshes/F_MED_RaggedRebel_Pack_AnimBP
target_skeleton=&/F_MED_RaggedRebel_Pack_Skeleton
graphs [
 AnimGraph {nodes [
   n1 = Control Rig!(
    Source=Input Pose()
    Alpha
    Is Sprinting:bool=Get Sprinting() -> (Sprinting:bool)
    Is Skydive:bool=Get bIsSkydiveFloating()
    Is IDLE:bool=Get isIDLE() -> (isIDLE:bool)
    Is Skydive Dive:bool=Get bIsSkydiveDiveMode()
    Is Skydive Parachute:bool=Get bIsParachuteOpened()
    Is Grindrail:bool=Get bIsOnGrindRail()
    Is Falling:bool=Get bIsFalling()
    Is Tactical Sprint:bool=Get bIsTacticalSprinting()
    Is Zipline:bool=Get OnZipline() -> (OnZipline:bool)
    Is Going Up:bool=Get GoingUp() -> (GoingUp:bool)
    Is Going Down:bool=Get GoingDown() -> (GoingDown:bool)
   ) -> (Pose) {
    alphaBoolBlend {blendOption=Linear}
    controlRigAssetReference {blueprintRigClass=&/F_MED_RaggedRebel_Pack_CtrlRig.F_MED_RaggedRebel_Pack_CtrlRig_C}
   }
   Output Pose(Result=@n1) {layerGroup=None}
  ]}
 EventGraph {nodes [
   Event BlueprintUpdateAnimation()
   > n1 = Sequence!(execute) -> (then_0, then_1, then_2:Exec)
   n2 = Get bIsMoving2D()
   n3 = Get bIsPlayingEmote()
   > Set isIDLE(isIDLE:bool=NOT Boolean(A=OR Boolean!(A=@n2, B=@n3, C:bool) -> (ReturnValue))) -> (Output_Get:bool)
   n4 = PlayerMovement(execute=@n1.then_2, AnimInstance=Self-Reference(), XDotProductCoefficient=100, YDotProductCoefficient=100, ZDotProductCoefficient=100)
   n5 = NOT Boolean(A=OR Boolean!(
     A=Get bIsParachuteOpened()
     B=Get bIsSkydiveDiveMode()
     C=Get bIsSkydiveFloating()
     D:bool=Get bIsFalling()
     E:bool=Get bIsJumping()
     F:bool=Get OnZipline()
    ) -> (ReturnValue))
   n6 = BreakVector(InVec=MakeVector(X=@n4, Y=@n4.YMovement, Z=@n4.ZMovement))
   > Set GoingUp(GoingUp:bool=AND Boolean(A="float > float"(A=@n6.Z, B=10), B=@n5)) -> (Output_Get:bool)
   > Set GoingDown(GoingDown:bool=AND Boolean(A="float < float"(A=@n6.Z, B=-10), B=@n5)) -> (Output_Get:bool)
   n7 = Get bIsPlayingEmote()
   Set Sprinting(execute=@n1.then_1, Sprinting:bool=AND Boolean(A=Get bIsMoving2D(), B=NOT Boolean(A=@n7))) -> (Output_Get:bool)
   n8 = Event BlueprintProcessAnimInputEvent()
   > n9 = Branch(Condition=IsNamed(self=@n8.AnimInputEvent, InEventName=aie.Player.Zipline.Entered))
   > Set OnZipline(OnZipline:bool=1) -> (Output_Get:bool)
   Branch(
    execute=@n9.else
    Condition=OR Boolean!(A=IsNamed(self=@n8.AnimInputEvent, InEventName=aie.Player.Zipline.Exited), B=Get bIsMoving2D(), C=Get bIsJumping()) -> (ReturnValue)
   )
   > Set OnZipline(OnZipline=false)
   Reroute Node(InputPin:Wildcard) -> (OutputPin:Wildcard)
  ]}
 NewMacro {nodes [
   n1 = Inputs!() -> (Start:Exec, state:bool=~)
   > n2 = Sequence!(execute) -> (then_0, then_1)
   > n3 = Branch(Condition=@n1)
   > Rotate!(execute, self, Target:double=5, Interp Speed:double=2, Delta time:double=0) -> (then)
   Delay(execute=@n3.else, Duration=0.2)
   > Rotate(Target=2, Interp Speed=2, Delta time=0)
   Outputs!(End:Exec=@n2.then_1)
  ]}
]
variables [
 GoingDown {type=boolean}
 GoingUp {type=boolean}
 isIDLE {type=boolean}
 OnZipline {type=boolean}
 Sprinting {type=boolean}
]
